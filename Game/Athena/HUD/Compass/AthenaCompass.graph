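# AthenaCompass is the Athena HUD compass widget, a Widget Blueprint derived from the native AthenaCompassBase class.
# Its root Image_Compass is a wide, short image strip drawn with the M_Compass_Inst material instance, alongside a TutorialHighlight widget.
# The event graph is empty and no variables are declared, so the Blueprint supplies visuals only and leaves any behavior to the parent class.
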WidgetBlueprint /Game/Athena/HUD/Compass/AthenaCompass
graphs [EventGraph {links [], nodes []}]
variables []
widgets {
 info {inheritedWidgetCount=2, parentClass=/Script/FortniteUI.AthenaCompassBase, rootWidgetClass=&Image, widgetCount=2}
 widgets [
  Image_Compass {
   widgetClassPath=&Image
   bInherited=true
   widget=Image_Compass
   properties {brush {
     bIsDynamicallyLoaded=false
     drawAs=Image
     imageType=NoImage
     mirroring=NoMirror
     resourceName=None
     resourceObject=&/M_Compass_Inst
     tiling=NoTile
     imageSize=<768 60>
     margin=<ltrb 0 0 0 0>
     outlineSettings {
      bUseBrushTransparency=false
      roundingType=HalfHeightRadius
      width=0
      color {colorUseRule=UseColor_Specified, specifiedColor=<rgba 0 0 0 0>}
      cornerRadii=<0 0 0 0>
     }
     tintColor {colorUseRule=UseColor_Specified, specifiedColor=<rgba 1 1 1 1>}
     uVRegion {bIsValid=false, max=<0 0>, min=<0 0>}
    }}
  }
  TutorialHighlight {widgetClassPath=&TutorialHighlight}
 ]
}
