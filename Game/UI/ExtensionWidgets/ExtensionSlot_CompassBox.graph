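# HUD extension slot container that hosts the Athena compass and stacks extension widgets with it in a vertical box.
# On extension widget creation, it adds the widget to Compass_Box with fill alignment, unless a matching child already exists.
# WidgetContainsChildClass performs that duplicate check, breaking at the first child whose class is the same as or derives from the new one.
# The built-in compass carries HUD tags for contextual hiding.
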WidgetBlueprint /Game/UI/ExtensionWidgets/ExtensionSlot_CompassBox
graphs [
 EventGraph {nodes [
   n1 = Event OnExtensionWidgetCreated!() -> (OutputDelegate, then, InUserWidget:User Widget Object Reference, InExtensionSlot:EUIExtensionSlot Enum=Invalid) {is=title K2Node_Event}
   > n2 = Branch(Condition=WidgetContainsChildClass!(
     WidgetToCheckChildren:Panel Widget Object Reference=Get Compass_Box() -> (Compass_Box:Vertical Box Object Reference)
     InWidgetClass:Widget Class Reference=Get Class(Object=@n1.InUserWidget) -> (ReturnValue:User Widget Class Reference)
    ) -> (bContainsChild:bool))
   n3 = AddChildToVerticalBox(execute=@n2.else, self=Get Compass_Box(), Content=@n1.InUserWidget)
   > SetHorizontalAlignment(self:Vertical Box Slot Object Reference=@n3, InHorizontalAlignment=HAlign_Fill)
   > SetVerticalAlignment(self:Vertical Box Slot Object Reference=@n3, InVerticalAlignment=VAlign_Fill)
  ]}
 WidgetContainsChildClass {nodes [
   n1 = WidgetContainsChildClass!() -> (then, WidgetToCheckChildren:Panel Widget Object Reference, InWidgetClass:Widget Class Reference) {is=title K2Node_FunctionEntry}
   > n2 = Is Valid(InputObject=@n1)
   > n3 = Branch(Condition=IsValidClass!(Class:Object Class Reference=@n1.InWidgetClass) -> (ReturnValue:bool))
   > n4 = For Each Loop with Break!(
    Exec:Exec
    Array:Array of Widget Object References=GetAllChildren(self=Get WidgetToCheckChildren() -> (WidgetToCheckChildren:Panel Widget Object Reference))
    Break:Exec=@n6.then
   ) -> (LoopBody:Exec, Array Element:Widget Object Reference, Array Index:int, Completed:Exec)
   Return Node!(execute=[@n4.Completed, @n3.else, @n2.Is Not Valid], bContainsChild:bool=Get bLoopFoundChild() -> (bLoopFoundChild:bool))
   > n5 = Set bLoopFoundChild(bLoopFoundChild:bool=ClassIsChildOf(TestClass=Get Class(Object=@n4) -> (ReturnValue:Widget Class Reference), ParentClass=@n1.InWidgetClass)) -> (Output_Get:bool)
   > n6 = Branch(Condition=@n5)
  ]}
]
variables []
widgets {
 info {inheritedWidgetCount=0, parentClass=/Script/FortniteUI.FortExtensionWidgetContainer, rootWidgetClass=&VerticalBox, widgetCount=2}
 widgets [
  Compass_Box {widgetClassPath=&VerticalBox, bIsVariable=true, widget=Compass_Box}
  AthenaCompass {
   widgetClassPath=/Game/Athena/HUD/Compass/AthenaCompass
   parent=Compass_Box
   widget=AthenaCompass
   properties {hUDElementTag {gameplayTags [
      {tagName=HUD.Athena.HideForPostGame}
      {tagName=HUD.Athena.HideForCreative}
      {tagName=HUD.Athena.HideForPlayground}
      {tagName=HUD.Athena.HideForCreativeLTM}
      {tagName=HUD.Athena.HideforRadial}
      {tagName=HUD.Hide.WhileLoading}
      {tagName=HUD.Athena.HideOnSpectate}
      {tagName=HUD.Athena.HideForPostGameSpectate}
      {tagName=HUD.Athena.Hide.WithMap}
      {tagName=HUD.Athena.HideForConversation}
      {tagName=HUD.Event.HideForSpecialEvent}
      {tagName=HUD.Hide.CreativeJoiningLobby}
     ]}}
   slot {class=VerticalBoxSlot, properties {horizontalAlignment=HAlign_Center, verticalAlignment=VAlign_Top}}
  }
 ]
}
