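# Reticle extension slot widget, derived from FortReticleExtensionContainer, that hosts dynamically created extension widgets in a root Overlay named Content.
# Both the general extension-created and reticle-extension-created events pass their widget to a shared AddChild function, ignoring the slot parameter.
# AddChild places each widget in the Overlay, centered horizontally and vertically.
# The reticle-extension-removed event removes its widget from that Overlay.
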WidgetBlueprint /Game/UI/ExtensionWidgets/ExtensionSlot_Reticle
graphs [
 AddChild {nodes [
   n1 = AddChild!() -> (then, ChildWidget:Widget Object Reference) {is=title K2Node_FunctionEntry}
   > n2 = AddChildToOverlay!(execute, self:Overlay Object Reference=Get Content() -> (Content:Overlay Object Reference), Content:Widget Object Reference=@n1) -> (then, ReturnValue:Overlay Slot Object Reference)
   > SetVerticalAlignment(self=@n2, InVerticalAlignment=VAlign_Center)
   > SetHorizontalAlignment(self=@n2, InHorizontalAlignment=HAlign_Center)
  ]}
 EventGraph {nodes [
   n1 = Event OnExtensionWidgetCreated!() -> (OutputDelegate, then, InUserWidget:User Widget Object Reference, InExtensionSlot:EUIExtensionSlot Enum=Invalid) {is=title K2Node_Event}
   > AddChild!(execute, self:Self Object Reference, ChildWidget:Widget Object Reference=@n1.InUserWidget) -> (then) {is=title K2Node_CallFunction}
   n2 = Event OnReticleExtensionWidgetCreated!() -> (OutputDelegate, then, Widget:User Widget Object Reference) {is=title K2Node_Event}
   > AddChild(ChildWidget=@n2.Widget)
   n3 = Event OnReticleExtensionWidgetRemoved!() -> (OutputDelegate, then, Widget:User Widget Object Reference) {is=title K2Node_Event}
   > RemoveChild(self=Get Content(), Content=@n3.Widget)
  ]}
]
variables []
widgets {
 info {inheritedWidgetCount=0, parentClass=/Script/FortniteUI.FortReticleExtensionContainer, rootWidgetClass=&Overlay, widgetCount=1}
 widgets [Content {widgetClassPath=&Overlay, bIsVariable=true, widget=Content}]
}
