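# ExtensionSlot_Secondary is a Fortnite UI extension widget container, named as the secondary slot, that hosts extension widgets created at runtime.
# A root ScaleBox with user-specified stretch and ignored inherited scale holds a Content overlay that fills it.
# When OnExtensionWidgetCreated fires, the graph adds the new widget to that overlay and sets it to fill both horizontally and vertically.
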WidgetBlueprint /Game/UI/ExtensionWidgets/ExtensionSlot_Secondary
graphs [EventGraph {nodes [
   n1 = Event OnExtensionWidgetCreated!() -> (OutputDelegate, then, InUserWidget:User Widget Object Reference, InExtensionSlot:EUIExtensionSlot Enum=Invalid) {is=title K2Node_Event}
   > n2 = AddChildToOverlay!(execute, self:Overlay Object Reference=Get Content() -> (Content:Overlay Object Reference), Content:Widget Object Reference=@n1.InUserWidget) -> (then, ReturnValue:Overlay Slot Object Reference)
   > SetVerticalAlignment(self=@n2, InVerticalAlignment=VAlign_Fill)
   > SetHorizontalAlignment(self=@n2, InHorizontalAlignment=HAlign_Fill)
  ]}]
variables [added {type=boolean}]
widgets {
 info {inheritedWidgetCount=0, parentClass=/Script/FortniteUI.FortExtensionWidgetContainer, rootWidgetClass=&ScaleBox, widgetCount=2}
 widgets [
  ScaleBox_0 {widgetClassPath=&ScaleBox, widget=ScaleBox_0, properties {ignoreInheritedScale=true, stretch=UserSpecified}}
  Content {
   widgetClassPath=&Overlay
   bIsVariable=true
   parent=ScaleBox_0
   widget=Content
   slot {class=ScaleBoxSlot, properties {horizontalAlignment=HAlign_Fill, verticalAlignment=VAlign_Fill}}
  }
 ]
}
