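# Extension slot widget, derived from FortExtensionWidgetContainer, that hosts HUD extension widgets shown under the reticle.
# An overlay holds a horizontally centered vertical box containing a top-aligned prioritized widget switcher with a zero transition duration.
# When an extension widget is created, the event graph adds it as a child of that switcher, without checking the slot enum.
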WidgetBlueprint /Game/UI/ExtensionWidgets/ExtensionSlot_UnderReticle
graphs [EventGraph {nodes [
   n1 = Event OnExtensionWidgetCreated!() -> (OutputDelegate, then, InUserWidget:User Widget Object Reference, InExtensionSlot:EUIExtensionSlot Enum=Invalid) {is=title K2Node_Event}
   > AddChild(
    self=Get WidgetSwitcher_UnderReticle() -> (WidgetSwitcher_UnderReticle:Fort Prioritized Widget Switcher Object Reference)
    Content=@n1.InUserWidget
   )
  ]}]
variables []
widgets {
 info {inheritedWidgetCount=0, parentClass=/Script/FortniteUI.FortExtensionWidgetContainer, rootWidgetClass=&Overlay, widgetCount=3}
 widgets [
  Overlay_Root {widgetClassPath=&Overlay, bIsVariable=true, widget=Overlay_Root, properties {bOverride_Cursor=true}}
  VerticalBox_0 {
   widgetClassPath=&VerticalBox
   parent=Overlay_Root
   widget=VerticalBox_0
   slot {class=OverlaySlot, properties {horizontalAlignment=HAlign_Center}}
  }
  WidgetSwitcher_UnderReticle {
   widgetClassPath=/Script/FortniteUI.FortPrioritizedWidgetSwitcher
   bIsVariable=true
   parent=VerticalBox_0
   widget=WidgetSwitcher_UnderReticle
   properties {transitionDuration=0}
   slot {class=VerticalBoxSlot, properties {horizontalAlignment=HAlign_Center, verticalAlignment=VAlign_Top}}
  }
 ]
}
