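# A Fortnite UI extension widget container that, judging by its name, hosts extension widgets in the slot under squad info.
# Its only widget is a root vertical box named Content.
# When an extension widget is created, the event graph adds it to Content as a new child with fill alignment on both axes, without checking the extension slot value.
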WidgetBlueprint /Game/UI/ExtensionWidgets/ExtensionSlot_UnderSquadInfo
graphs [EventGraph {nodes [
   n1 = Event OnExtensionWidgetCreated!() -> (OutputDelegate, then, InUserWidget:User Widget Object Reference, InExtensionSlot:EUIExtensionSlot Enum=Invalid) {is=title K2Node_Event}
   > n2 = AddChildToVerticalBox(self=Get Content() -> (Content:Vertical Box Object Reference), Content=@n1.InUserWidget)
   > SetHorizontalAlignment(self:Vertical Box Slot Object Reference=@n2, InHorizontalAlignment=HAlign_Fill)
   > SetVerticalAlignment(self:Vertical Box Slot Object Reference=@n2, InVerticalAlignment=VAlign_Fill)
  ]}]
variables []
widgets {
 info {inheritedWidgetCount=0, parentClass=/Script/FortniteUI.FortExtensionWidgetContainer, rootWidgetClass=&VerticalBox, widgetCount=1}
 widgets [Content {widgetClassPath=&VerticalBox, bIsVariable=true, widget=Content}]
}
