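# HUD extension slot container for the upper-center screen area, hosting phase alerts, storm surge, countdown, prompt, event, duel and collection toast widgets.
# These share a prioritized widget switcher, using priority and contest rules plus HUD tags for hiding.
# Extension widgets created at runtime join the same switcher.
# When playlist data is ready, spatial gameplay experiences append joining-lobby tags to the phase alert unless its starter-island display flag is set.
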WidgetBlueprint /Game/UI/ExtensionWidgets/ExtensionSlot_UpperCenter
graphs [EventGraph {nodes [
   Event Construct()
   > Parent: Construct()
   > n1 = PlaylistDataReadyAsync!(execute:Exec) -> (
    then:Exec
    OnReady:Exec
    GameState:Fort Game State Athena Object Reference
    Playlist:Fort Playlist Object Reference
    PlaylistContextTags:"Gameplay Tag Container Structure (by ref)"
   ) {is=title K2Node_AsyncAction}
   n2 = Get GamePhaseAlert() -> (GamePhaseAlert:Athena Game Phase Change Widget Object Reference)
   n3 = GetValueAsBoolWithContext!(
    Input:"Scalable Float Structure (by ref)"=Get ShouldDisplayGamePhaseUIOnStarterIsland!(self:Athena Game Phase Change Widget Object Reference=@n2) -> (ShouldDisplayGamePhaseUIOnStarterIsland:Scalable Float Structure)
    Level:float=0
   ) -> (ReturnValue:bool)
   Branch(execute=@n1.OnReady, Condition=AND Boolean(A=IsSpatialGameplayExperience!() -> (ReturnValue:bool), B=NOT Boolean(A=@n3)))
   > AppendHUDElementTags!(
    execute
    self:Fort HUDElement Widget Object Reference=@n2
    AdditionalTags:"Gameplay Tag Container Structure (by ref)"=Get JoiningLobbyTagContainer!(self:Athena Game Phase Change Widget Object Reference=@n2) -> (JoiningLobbyTagContainer:Gameplay Tag Container Structure)
   ) -> (then)
   n4 = Event OnExtensionWidgetCreated!() -> (OutputDelegate, then, InUserWidget:User Widget Object Reference, InExtensionSlot:EUIExtensionSlot Enum=Invalid) {is=title K2Node_Event}
   > AddChild(self=Get WidgetSwitcher_TopMid() -> (WidgetSwitcher_TopMid:Fort Prioritized Widget Switcher Object Reference), Content=@n4.InUserWidget)
  ]}]
variables []
widgets {
 info {inheritedWidgetCount=0, parentClass=/Script/FortniteUI.FortExtensionWidgetContainer, rootWidgetClass=&Overlay, widgetCount=10}
 widgets [
  Overlay_Root {widgetClassPath=&Overlay, widget=Overlay_Root, properties {bOverride_Cursor=true}}
  VerticalBox_0 {
   widgetClassPath=&VerticalBox
   parent=Overlay_Root
   widget=VerticalBox_0
   slot {class=OverlaySlot, properties {horizontalAlignment=HAlign_Center}}
  }
  WidgetSwitcher_TopMid {
   widgetClassPath=/Script/FortniteUI.FortPrioritizedWidgetSwitcher
   bIsVariable=true
   parent=VerticalBox_0
   widget=WidgetSwitcher_TopMid
   properties {transitionDuration=0}
   slot {class=VerticalBoxSlot, properties {horizontalAlignment=HAlign_Center, verticalAlignment=VAlign_Top}}
  }
  GamePhaseAlert {
   widgetClassPath=/Game/Athena/HUD/Phase/AthenaGamePhaseChangeWidget
   bIsVariable=true
   parent=WidgetSwitcher_TopMid
   widget=GamePhaseAlert
   properties {
    visibility=Collapsed
    elementTag {tagName=HUD.Athena.GamePhase.AlertMessage}
    hUDElementTag {gameplayTags [
      {tagName=Athena.HideForCapture}
      {tagName=HUD.LTM.HideForTryHard}
      {tagName=HUD.Event.HideForSpecialEvent}
      {tagName=HUD.Athena.HideForPostGame}
     ]}
    prioritizationData {priority=Priority_2, whenContesting=QueueBehind}
   }
   slot {class=WidgetSwitcherSlot, properties {verticalAlignment=VAlign_Top}}
  }
  AthenaStormSurgePhase {
   widgetClassPath=/Game/Athena/HUD/AthenaStormSurgePhase
   parent=WidgetSwitcher_TopMid
   widget=AthenaStormSurgePhase
   properties {
    elementTag {tagName=HUD.Special.StormSurgePhase}
    hUDElementTag {gameplayTags [{tagName=HUD.Athena.HideForPostGameSpectate}, {tagName=HUD.Athena.HideForPostGame}, {tagName=HUD.Hide.CreativeJoiningLobby}]}
    prioritizationData {priority=Priority_2, whenContesting=QueueBehind}
   }
   slot {class=WidgetSwitcherSlot}
  }
  AthenaCountdown {
   widgetClassPath=/Game/Athena/HUD/AthenaCountdown
   parent=WidgetSwitcher_TopMid
   widget=AthenaCountdown
   properties {
    elementTag {tagName=HUD.Special.Countdown}
    hUDElementTag {gameplayTags [
      {tagName=Athena.HideForCapture}
      {tagName=HUD.Athena.HideForPostGame}
      {tagName=HUD.Athena.HideForPostGameSpectate}
      {tagName=HUD.Hide.CreativeJoiningLobby}
     ]}
    prioritizationData {priority=Priority_2, whenContesting=StompOther}
    renderTransform {angle=0, scale=<1 1>, shear=<0 0>, translation=<0 76>}
   }
   slot {class=WidgetSwitcherSlot, properties {verticalAlignment=VAlign_Bottom}}
  }
  VolumePromptWidget {
   widgetClassPath=/Game/UI/Foundation/HUDTagPrompt/HUDTagPromptWidget
   parent=WidgetSwitcher_TopMid
   widget=VolumePromptWidget
   properties {hUDElementTag {gameplayTags [{tagName=HUD.LTM.HideForTryHard}, {tagName=Athena.HideForCapture}]}}
   slot {class=WidgetSwitcherSlot, properties {horizontalAlignment=HAlign_Center, verticalAlignment=VAlign_Top}}
  }
  AthenaSpecialEventStartedOverlay {
   widgetClassPath=/Game/Athena/HUD/Discoverability/AthenaSpecialEventStartedOverlay
   parent=WidgetSwitcher_TopMid
   widget=AthenaSpecialEventStartedOverlay
   properties {
    elementTag {tagName=HUD.Special.SpecialEventStarted}
    hUDElementTag {gameplayTags [{tagName=Athena.HideForCapture}, {tagName=HUD.Hide.CreativeJoiningLobby}]}
   }
   slot {class=WidgetSwitcherSlot}
  }
  AthenaDuelOverlay {
   widgetClassPath=/Game/Athena/HUD/Duel/AthenaDuelOverlay
   parent=WidgetSwitcher_TopMid
   widget=AthenaDuelOverlay
   properties {
    elementTag {tagName=HUD.Athena.DuelOverlay}
    hUDElementTag {gameplayTags [
      {tagName=Athena.HideForCapture}
      {tagName=HUD.Athena.HideForPostGame}
      {tagName=HUD.Athena.HideForPostGameSpectate}
      {tagName=HUD.Hide.CreativeJoiningLobby}
     ]}
   }
   slot {class=WidgetSwitcherSlot}
  }
  BP_AthenaCollectionItemToast {
   widgetClassPath=/Game/Athena/HUD/Collections/BP_AthenaCollectionItemToast
   parent=WidgetSwitcher_TopMid
   widget=BP_AthenaCollectionItemToast
   properties {hUDElementTag {gameplayTags [{tagName=Athena.HideForCapture}, {tagName=HUD.Hide.CreativeJoiningLobby}]}}
   slot {class=WidgetSwitcherSlot}
  }
 ]
}
