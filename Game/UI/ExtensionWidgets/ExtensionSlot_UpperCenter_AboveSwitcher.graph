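# Fort extension widget container for the upper-center HUD area, stacking a vertical box above a prioritized widget switcher.
# Extension widgets for the AboveQuestsMobileUI slot go into the box, those for AboveUpperCenterPrioritized into the switcher, both fill-aligned.
# It registers as a player-specific widget host for the AboveUpperCenter slot tag on construct and unregisters on destruct.
# Host content with that exact tag is added to or removed from the switcher.
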WidgetBlueprint /Game/UI/ExtensionWidgets/ExtensionSlot_UpperCenter_AboveSwitcher
graphs [
 AddContentToHostWidget {nodes [
   AddContentToHostWidget!() -> (then, WidgetHostTag:Gameplay Tag Structure, ContentWidget:Widget Object Reference, Padding:Margin Structure, Fill:bool=~) {is=title K2Node_FunctionEntry}
   > Branch(Condition=MatchesTag!(
     TagOne:Gameplay Tag Structure=Get WidgetHostTag() -> (WidgetHostTag:Gameplay Tag Structure)
     TagTwo:Gameplay Tag Structure=Get AboveUpperCenterSlotTag() -> (AboveUpperCenterSlotTag:Gameplay Tag Structure)
     bExactMatch:bool=true
    ) -> (ReturnValue:bool))
   > AddChild(
    self=Get PrioritizedSwitcher_UpperCenter_AboveSwitcher() -> (PrioritizedSwitcher_UpperCenter_AboveSwitcher:Fort Prioritized Widget Switcher Object Reference)
    Content=Get ContentWidget() -> (ContentWidget:Widget Object Reference)
   )
  ]}
 EventGraph {nodes [
   n1 = Event Add Content To Host Widget!() -> (
    OutputDelegate
    then
    WidgetHostTag:"Gameplay Tag Structure (by ref)"
    ContentWidget:Widget Object Reference
    Padding:"Margin Structure (by ref)"
    bFill:bool
   ) {is=title K2Node_Event}
   > AddContentToHostWidget!(
    execute
    self
    WidgetHostTag:Gameplay Tag Structure=@n1.WidgetHostTag
    ContentWidget:Widget Object Reference=@n1.ContentWidget
    Padding:Margin Structure=@n1.Padding
    Fill:bool=@n1.bFill
   ) -> (then) {is=title K2Node_CallFunction}
   Event Construct()
   n2 = Get AboveUpperCenterSlotTag()
   > RegisterWidgetHost!(execute, WidgetHost:Widget Object Reference, WidgetHostTag:"Gameplay Tag Structure (by ref)"=@n2, IsPlayerSpecific:bool=true) -> (then)
   Event Destruct()
   > UnregisterWidgetHost!(execute, WidgetHost:Widget Object Reference, WidgetHostTag:"Gameplay Tag Structure (by ref)"=@n2) -> (then)
  ]}
 OnExtensionWidgetCreated {nodes [
   OnExtensionWidgetCreated!() -> (then, InUserWidget:User Widget Object Reference, InExtensionSlot:EUIExtensionSlot Enum=Invalid) {is=title K2Node_FunctionEntry}
   > n1 = Branch(Condition="Equal (Enum)"(
     A:EUIExtensionSlot Enum=Get InExtensionSlot() -> (InExtensionSlot:EUIExtensionSlot Enum=Invalid)
     B:EUIExtensionSlot Enum=Get AboveQuestsMobileUI() -> (AboveQuestsMobileUI:EUIExtensionSlot Enum=Invalid)
    ))
   > n2 = AddChild(
    self=Get VBox_PrimarySlot() -> (VBox_PrimarySlot:Vertical Box Object Reference)
    Content=Get InUserWidget() -> (InUserWidget:User Widget Object Reference)
   )
   > n3 = Cast To VerticalBoxSlot!(execute, Object=@n2) -> (then, CastFailed, AsVertical Box Slot:Vertical Box Slot Object Reference) {is=title K2Node_DynamicCast}
   > SetVerticalAlignment(self:Vertical Box Slot Object Reference=@n3, InVerticalAlignment=VAlign_Fill)
   > SetHorizontalAlignment(self:Vertical Box Slot Object Reference=@n3, InHorizontalAlignment=HAlign_Fill)
   Branch(
    execute=@n1.else
    Condition="Equal (Enum)"(
     A=Get InExtensionSlot() -> (InExtensionSlot=Invalid)
     B=Get AboveUpperCenterPrioritized() -> (AboveUpperCenterPrioritized:EUIExtensionSlot Enum=Invalid)
    )
   )
   > n4 = AddChild(self=Get PrioritizedSwitcher_UpperCenter_AboveSwitcher(), Content=Get InUserWidget())
   > n5 = Cast To WidgetSwitcherSlot!(execute, Object=@n4) -> (then, CastFailed, AsWidget Switcher Slot:Widget Switcher Slot Object Reference) {is=title K2Node_DynamicCast}
   > SetVerticalAlignment(self:Widget Switcher Slot Object Reference=@n5, InVerticalAlignment=VAlign_Fill)
   > SetHorizontalAlignment(self:Widget Switcher Slot Object Reference=@n5, InHorizontalAlignment=HAlign_Fill)
  ]}
 RemoveContentFromWidgetHost {nodes [
   Remove Content From Host Widget!() -> (
    then
    WidgetHostTag:"Gameplay Tag Structure (by ref)"
    ContentWidget:Widget Object Reference
    RemovalType:EWidgetSlottingRemovalType Enum=Unparent
   ) {is=title K2Node_FunctionEntry}
   > n1 = Branch(Condition=MatchesTag(TagOne=Get WidgetHostTag() -> (WidgetHostTag:"Gameplay Tag Structure (by ref)"), TagTwo=Get AboveUpperCenterSlotTag(), bExactMatch=true))
   > n2 = RemoveChild(self=Get PrioritizedSwitcher_UpperCenter_AboveSwitcher(), Content=Get ContentWidget())
   > Return Node!(execute, ReturnValue:bool=@n2)
   Return Node(execute=@n1.else, ReturnValue=false)
  ]}
]
variables [
 AboveQuestsMobileUI {type=EUIExtensionSlot, default=AboveQuestsMobileUI}
 AboveUpperCenterPrioritized {type=EUIExtensionSlot, default=AboveUpperCenterPrioritized}
 AboveUpperCenterSlotTag {type=GameplayTag, default {tagName=DynamicUI.HUD.Fort.Athena.AboveUpperCenter.Slot}}
]
widgets {
 info {inheritedWidgetCount=0, parentClass=/Script/FortniteUI.FortExtensionWidgetContainer, rootWidgetClass=&VerticalBox, widgetCount=3}
 widgets [
  VerticalBox_0 {widgetClassPath=&VerticalBox, widget=VerticalBox_0}
  VBox_PrimarySlot {widgetClassPath=&VerticalBox, bIsVariable=true, parent=VerticalBox_0, widget=VBox_PrimarySlot, slot {class=VerticalBoxSlot}}
  PrioritizedSwitcher_UpperCenter_AboveSwitcher {
   widgetClassPath=/Script/FortniteUI.FortPrioritizedWidgetSwitcher
   bIsVariable=true
   parent=VerticalBox_0
   widget=PrioritizedSwitcher_UpperCenter_AboveSwitcher
   properties {transitionCurveType=Linear, transitionDuration=0}
   slot {class=VerticalBoxSlot}
  }
 ]
}
