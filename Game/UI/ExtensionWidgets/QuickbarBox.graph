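# QuickbarBox is a FortExtensionWidgetContainer widget that hosts extension widgets in a horizontal row, apparently for the quickbar area of the UI.
# A SizeBox root fixes the height and sets a small minimum width around a HorizontalBox named Content.
# When an extension widget is created, the event graph adds it to Content and sets its slot to fill both horizontally and vertically.
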WidgetBlueprint /Game/UI/ExtensionWidgets/QuickbarBox
graphs [EventGraph {nodes [
   n1 = Event OnExtensionWidgetCreated!() -> (OutputDelegate, then, InUserWidget:User Widget Object Reference, InExtensionSlot:EUIExtensionSlot Enum=Invalid) {is=title K2Node_Event}
   > n2 = AddChildToHorizontalBox!(
    execute
    self:Horizontal Box Object Reference=Get Content() -> (Content:Horizontal Box Object Reference)
    Content:Widget Object Reference=@n1.InUserWidget
   ) -> (then, ReturnValue:Horizontal Box Slot Object Reference)
   > SetHorizontalAlignment(self:Horizontal Box Slot Object Reference=@n2, InHorizontalAlignment=HAlign_Fill)
   > SetVerticalAlignment(self:Horizontal Box Slot Object Reference=@n2, InVerticalAlignment=VAlign_Fill)
  ]}]
variables []
widgets {
 info {inheritedWidgetCount=0, parentClass=/Script/FortniteUI.FortExtensionWidgetContainer, rootWidgetClass=&SizeBox, widgetCount=2}
 widgets [
  SizeBox_1 {
   widgetClassPath=&SizeBox
   widget=SizeBox_1
   properties {bOverride_HeightOverride=true, bOverride_MinDesiredWidth=true, heightOverride=160, minDesiredWidth=10}
  }
  Content {widgetClassPath=&HorizontalBox, bIsVariable=true, parent=SizeBox_1, widget=Content, slot {class=SizeBoxSlot}}
 ]
}
