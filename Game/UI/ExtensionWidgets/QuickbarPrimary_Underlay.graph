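# QuickbarPrimary_Underlay is a Fortnite UI extension widget container, named as the underlay for the primary quickbar, that hosts extension widgets.
# Its widget tree is a single root Overlay named Content.
# When an extension widget is created, the OnExtensionWidgetCreated event adds that widget as a child of the Content overlay, without checking the extension slot it receives.
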WidgetBlueprint /Game/UI/ExtensionWidgets/QuickbarPrimary_Underlay
graphs [EventGraph {nodes [
   n1 = Event OnExtensionWidgetCreated!() -> (OutputDelegate, then, InUserWidget:User Widget Object Reference, InExtensionSlot:EUIExtensionSlot Enum=Invalid) {is=title K2Node_Event}
   > AddChild(self=Get Content() -> (Content:Overlay Object Reference), Content=@n1.InUserWidget)
  ]}]
variables []
widgets {
 info {inheritedWidgetCount=0, parentClass=/Script/FortniteUI.FortExtensionWidgetContainer, rootWidgetClass=&Overlay, widgetCount=1}
 widgets [Content {widgetClassPath=&Overlay, bIsVariable=true, widget=Content}]
}
